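; Blueprint for a ground smash decal, judging by its name, in the AnnoyedMop VFX decal assets.
; BeginPlay, Tick and the construction script are empty, so the graphs shown add no behavior of their own.
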(event EventBeginPlay)

(event EventTick
  (params (DeltaSeconds :type Float%20%28single-precision%29)))

(fn ConstructionScript)
