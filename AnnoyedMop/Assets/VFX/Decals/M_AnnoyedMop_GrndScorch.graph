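# Translucent deferred decal material for a ground scorch mark with an optional glowing emissive.
# Two panning noise layers and a godray noise in radial coordinates combine with scorch textures to drive the emissive, tinted by a parameter or particle color.
# Opacity combines the decal texture mask with decal lifetime fading.
# Named reroutes define UVs rotated by object world position for variation, plus an optional noise-broken vertical fade from height above the decal origin.
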Material /AnnoyedMop/Assets/VFX/Decals/M_AnnoyedMop_GrndScorch
graph {
 properties {blendMode=BLEND_Translucent, materialDomain=MD_DeferredDecal, pixelDepthOffsetMode=PDOM_Legacy, usedShadingModels=Default Lit}
 nodes [
  n1 = ScalarParameter() {parameterName=NoisePhasing}
  n2 = Time()
  n3 = ScalarParameter() {defaultValue=1, parameterName=NoiseTile}
  n4 = NamedRerouteUsage()
  n5 = TextureCoordinate()
  n6 = MaterialFunctionCall(Vector or UVs=Add(A=Multiply(A=@n4) {constB=0.05}, B=Subtract(A=@n5) {constB=0.5})) {materialFunction=&VectorToRadialValue}
  n7 = Multiply(A=@n6.Radial Coordinates) {constB=3}
  n8 = Multiply(A=Time(), B=ScalarParameter() {defaultValue=0.25, parameterName=PanSpeed})
  n9 = TextureSample(UVs=Panner(Coordinate=Multiply(A=@n3, B=@n7), Time=@n8) {speedX=0.005, speedY=-0.02}) {samplerType=SAMPLERTYPE_Masks, texture=/Game/Global/Textures/Default/Noise/T_CubeNoise}
  n10 = Multiply(A=@n3, B=Multiply(A=@n6.Radial Coordinates) {constB=4})
  n11 = Multiply(
   A=@n9.B
   B=TextureSample(UVs=Panner(Coordinate=@n10, Time=@n8) {speedX=-0.015, speedY=0.03}) {samplerType=SAMPLERTYPE_Masks, texture=&T_CubeNoise}.B
  )
  n12 = TextureSample(UVs=@n7) {texture=/Game/Global/Textures/Default/Noise/T_godray_Noise}
  n13 = NamedRerouteUsage()
  n14 = TextureSampleParameter2D(UVs=@n13) {
   parameterName=DecalTexture
   samplerType=SAMPLERTYPE_LinearColor
   texture=/Game/Effects/Fort_Effects/Gameplay/SmokeDustExplosions/Textures/T_GroundChar_01
  }
  n15 = StaticSwitchParameter(
   True=@n14.G
   False=TextureSampleParameter2D(UVs=@n13) {parameterName=EmissiveTextureA, texture=/Game/Effects/Fort_Effects/Gameplay/SmokeDustExplosions/Textures/T_Decal_Scorch_Hole}.G
  ) {defaultValue=true, parameterName=UseDecalTexture}
  n16 = Add(A=Add(A=Multiply(A=@n1, B=@n2), B=@n11), B=Add(A=@n12.G, B=@n15))
  Add(A=Multiply(A=Sine(None=@n16)) {constB=0.5}) {constB=0.5}
  ComponentMask(None=NamedRerouteUsage()) {g=true, r=true}
  DecalLifetimeOpacity()
  n17 = NamedRerouteUsage()
  n18 = TextureSample(UVs=@n17) {texture=&T_Decal_Scorch_Hole}
  n19 = Multiply(A=@n12.G, B=@n18.G)
  n20 = Multiply(A=@n9.R, B=@n11)
  n21 = Add(
   A=Multiply(A=Multiply(A=@n20, B=@n20) {constB=2}, B=ScalarParameter() {defaultValue=4, parameterName=NoiseMult}) {constB=4}
   B=ScalarParameter() {defaultValue=0.7, parameterName=NoiseAdd}
  ) {constB=0.7}
  n22 = ScalarParameter() {defaultValue=2, parameterName=NoiseExponent}
  n23 = Multiply(A=@n20, B=Multiply(A=Power(Base=@n21, Exp=@n22), B=@n18.G))
  n24 = TextureSample(UVs=@n17) {texture=&T_Decal_Scorch_Hole}
  n25 = Multiply(
   A=StaticSwitchParameter(True=ParticleColor().RGB, False=VectorParameter() {group=Emissive, parameterName=EmissiveEdge, defaultValue=<rgba 0.8 0.1474 0.0516 1>}.RGB) {parameterName=bParticleControlledColor}
   B=Multiply(A=Multiply(A=@n19, B=@n23), B=OneMinus(None=@n24.R))
  )
  n26 = Clamp(None=@n25)
  MaterialFunctionCall(Emissive=StaticSwitchParameter(
    True=MaterialFunctionCall(LDR Input=@n26, HDR Tint=@n25) {materialFunction=&DeriveHDRfromLDR}.RemappedRange
    False=Constant()
   ) {defaultValue=true, group=Timestamp, parameterName=Use Emissive?}) {materialFunction=&MF_FX_EmissiveAdjustments}
  n27 = DecalLifetimeOpacity()
  n28 = Saturate(None=Multiply(A=@n27.Opacity, B=ScalarParameter() {defaultValue=1, parameterName=FadeDelay}))
  n29 = ScalarParameter() {parameterName=DecalAffectOpacity}
  n30 = Multiply(A=Multiply(A=LinearInterpolate(B=@n28, Alpha=@n29) {constA=1}, B=@n15), B=ScalarParameter() {defaultValue=1, parameterName=OpacityMult})
  Multiply(A=Power(Base=Multiply(A=@n30, B=@n27.Opacity), Exp=LinearInterpolate(Alpha=@n27.Opacity) {constA=5}), B=NamedRerouteUsage())
  n31 = DotProduct(A=ObjectPositionWS(), B=Constant3Vector() {constant=<rgba 1 1 1 0>})
  NamedRerouteDeclaration(None=Rotator(Time=Divide(A=@n31) {constB=10})) {name=DecalUV, nodeColor=<rgba 1 0 0.8672 1>}
  n32 = ComponentMask(None=Subtract(A=WorldPosition().XYZ, B=ObjectPositionWS())) {b=true}
  n33 = TextureSample(UVs=TextureCoordinate() {uTiling=6, vTiling=6}) {texture=&T_godray_Noise}
  n34 = Power(Base=@n33.G) {constExponent=0.5}
  n35 = OneMinus(None=MaterialFunctionCall(
    Input=Multiply(A=@n32, B=@n34)
    Input Low=ScalarParameter() {defaultValue=20, group=VerticalFade, parameterName=FadeBegin}
    Input High=ScalarParameter() {defaultValue=50, group=VerticalFade, parameterName=FadeEnd}
   ) {materialFunction=&RemapValueRange}.Result)
  n36 = ScalarParameter() {group=VerticalFade, parameterName=UseVerticalGradient}
  NamedRerouteDeclaration(None=LinearInterpolate(B=Saturate(None=@n35), Alpha=@n36) {constA=1}) {name=VerticalGradient, nodeColor=<rgba 0 1 0.9766 1>}
  n37 = Divide(A=@n32) {constB=100}
  NamedRerouteDeclaration(None=LinearInterpolate(B=Multiply(A=@n37, B=@n33.G), Alpha=@n36) {constA=1, constAlpha=0}) {name=VerticalNoise, nodeColor=<rgba 0 1 0.6484 1>}
  ScalarParameter() {parameterName=Metallic}
  ScalarParameter() {parameterName=Specular}
  ScalarParameter() {defaultValue=1, parameterName=Roughness}
  SphereMask(A=@n6.Radial Coordinates, B=Constant() {r=0.5})
  StaticSwitchParameter(True=ParticleColor().A, False=@n27.Opacity) {parameterName=bParticleControlledColor}
  TextureCoordinate() {uTiling=3, vTiling=3}
  TextureCoordinate() {uTiling=4, vTiling=4}
  TextureCoordinate() {uTiling=4, vTiling=4}
  VectorParameter() {parameterName=DecalColor, defaultValue=<rgba 0 0 0 1>}
 ]
}
